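# Physics asset for the skydive variant of the female Headband_S adult body, providing collision and joints for its dynamic dress bones.
# Four box-body chains (left/right, a/b) hang from master capsules on the root, with tight swing and twist limits.
# Angularly free cross-links join both b chains near the top and each side's a and b tips.
# Leg capsules plus a root capsule and wide thin box supply body collision shapes.
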PhysicsAsset /BRCosmetics/Characters/Player/Female/Medium/Bodies/F_MED_Headband_S_Adult/Meshes/F_MED_Headband_S_Adult_Skydive_PHYS
bodies {
 calf_l [{shapeType=Capsule, length=16.4587, radius=9.9583, shapeName=calf_l_capsule, center=<-16.8777 -0.0002 -0.0001>, rotation=<rot 90 0 0>}]
 calf_r [{shapeType=Capsule, length=21.162, radius=6.0778, shapeName=calf_r_capsule, center=<19.401 0 -0.0001>, rotation=<rot 90 0 0>}]
 dyn_dress_left_a_01 [{shapeType=Box, extentX=11.5, extentY=4, extentZ=11, shapeName=dyn_dress_left_a_01_box, center=<5.5933 0 0>, rotation=<rot 0 0 49.2836>}]
 dyn_dress_left_a_02 [{shapeType=Box, extentX=11.5, extentY=6, extentZ=9.5, shapeName=dyn_dress_left_a_02_box, center=<5.5933 0 0>, rotation=<rot 0 0 55.1246>}]
 dyn_dress_left_a_03 [{shapeType=Box, extentX=11.5, extentY=4, extentZ=11, shapeName=dyn_dress_left_a_03_box, center=<5.5933 0 0>, rotation=<rot 0 0 38.797>}]
 dyn_dress_left_a_04 [{shapeType=Box, extentX=11.5, extentY=4, extentZ=9, shapeName=dyn_dress_left_a_04_box, center=<5.5933 0 0>, rotation=<rot 0 0 44.101>}]
 dyn_dress_left_a_05 [{shapeType=Box, extentX=11.5, extentY=4, extentZ=9, shapeName=dyn_dress_left_a_05_box, center=<5.5933 0 0>, rotation=<rot 0 0 29.3286>}]
 dyn_dress_left_a_06 [{shapeType=Box, extentX=11.5, extentY=4, extentZ=9, shapeName=dyn_dress_left_a_06_box, center=<5.5933 0 0>, rotation=<rot 0 0 28.3658>}]
 dyn_dress_left_a_master [{shapeType=Capsule, length=0.505, radius=0.505, shapeName=dyn_dress_left_a_master_capsule, rotation=<rot 0 -90 0>}]
 dyn_dress_left_b_01 [{shapeType=Box, extentX=11, extentY=3.5, extentZ=10, shapeName=dyn_dress_left_b_01_box, center=<5.4988 0.0002 0>}]
 dyn_dress_left_b_02 [{shapeType=Box, extentX=11, extentY=3.5, extentZ=10, shapeName=dyn_dress_left_b_02_box, center=<5.4988 0.0002 0>}]
 dyn_dress_left_b_03 [{shapeType=Box, extentX=11, extentY=3.5, extentZ=10, shapeName=dyn_dress_left_b_03_box, center=<5.4988 0.0002 0>}]
 dyn_dress_left_b_04 [{shapeType=Box, extentX=11, extentY=3.5, extentZ=10, shapeName=dyn_dress_left_b_04_box, center=<5.4988 0.0002 0>}]
 dyn_dress_left_b_05 [{shapeType=Box, extentX=11, extentY=3.5, extentZ=10, shapeName=dyn_dress_left_b_05_box, center=<5.4988 0.0002 0>}]
 dyn_dress_left_b_master [{shapeType=Capsule, length=0.505, radius=0.505, shapeName=dyn_dress_left_b_master_capsule, rotation=<rot 0 -90 0>}]
 dyn_dress_right_a_01 [{shapeType=Box, extentX=11.5, extentY=6, extentZ=9, shapeName=dyn_dress_right_a_01_box, center=<5.5933 0 0>, rotation=<rot 0 0 -57.4517>}]
 dyn_dress_right_a_02 [{shapeType=Box, extentX=11.5, extentY=6, extentZ=9, shapeName=dyn_dress_right_a_02_box, center=<5.5933 0 0>, rotation=<rot 0 0 -57.4524>}]
 dyn_dress_right_a_03 [{shapeType=Box, extentX=11.5, extentY=6, extentZ=9, shapeName=dyn_dress_right_a_03_box, center=<5.5933 0 0>, rotation=<rot 0 0 -59.9293>}]
 dyn_dress_right_a_04 [{shapeType=Box, extentX=11.5, extentY=6, extentZ=9, shapeName=dyn_dress_right_a_04_box, center=<5.5933 0 0>, rotation=<rot 0 0 -38.8162>}]
 dyn_dress_right_a_05 [{shapeType=Box, extentX=11.5, extentY=6, extentZ=9, shapeName=dyn_dress_right_a_05_box, center=<5.5933 0 0>, rotation=<rot 0 0 -32.1226>}]
 dyn_dress_right_a_06 [{shapeType=Box, extentX=11.5, extentY=6, extentZ=9, shapeName=dyn_dress_right_a_06_box, center=<5.5933 0 0>, rotation=<rot 0 0 -27.0552>}]
 dyn_dress_right_a_master [{shapeType=Capsule, length=0.505, radius=0.505, shapeName=dyn_dress_right_a_master_capsule, rotation=<rot 0 -90 0>}]
 dyn_dress_right_b_01 [{shapeType=Box, extentX=11, extentY=5.5, extentZ=10, shapeName=dyn_dress_right_b_01_box, center=<5.4988 0.0002 0>}]
 dyn_dress_right_b_02 [{shapeType=Box, extentX=11, extentY=5.5, extentZ=10, shapeName=dyn_dress_right_b_02_box, center=<5.4988 0.0002 0>}]
 dyn_dress_right_b_03 [{shapeType=Box, extentX=11, extentY=5.5, extentZ=10, shapeName=dyn_dress_right_b_03_box, center=<5.4988 0.0002 0>}]
 dyn_dress_right_b_04 [{shapeType=Box, extentX=11, extentY=5.5, extentZ=10, shapeName=dyn_dress_right_b_04_box, center=<5.4988 0.0002 0>}]
 dyn_dress_right_b_05 [{shapeType=Box, extentX=11, extentY=5.5, extentZ=10, shapeName=dyn_dress_right_b_05_box, center=<5.4988 0.0002 0>}]
 dyn_dress_right_b_master [{shapeType=Capsule, length=0.505, radius=0.505, shapeName=dyn_dress_right_b_master_capsule, rotation=<rot 0 -90 0>}]
 root [
  {shapeType=Capsule, length=27, radius=19.75, shapeName=root_capsule, rotation=<rot 90 0 0>}
  {shapeType=Box, extentX=100, extentY=100, extentZ=0.5, shapeName=root_box}
 ]
 thigh_l [{shapeType=Capsule, length=23.4383, radius=9.9017, shapeName=thigh_l_capsule, center=<-25.2796 -0.0002 0>, rotation=<rot 90 0 0>}]
 thigh_r [{shapeType=Capsule, length=27.1221, radius=6.8173, shapeName=thigh_r_capsule, center=<26.5115 -0.0004 0>, rotation=<rot 90 0 0>}]
}
constraints [
 {
  bone1Name=dyn_dress_right_b_01
  bone2Name=dyn_dress_right_b_master
  swing1LimitDegrees=15
  swing2LimitDegrees=5
  swing2Motion=Locked
  twistLimitDegrees=5
 }
 {
  bone1Name=dyn_dress_right_b_02
  bone2Name=dyn_dress_right_b_01
  swing1LimitDegrees=9
  swing2LimitDegrees=5
  swing2Motion=Locked
  twistLimitDegrees=5
 }
 {bone1Name=dyn_dress_right_b_03, bone2Name=dyn_dress_right_b_02, swing1LimitDegrees=9, swing2LimitDegrees=5, twistLimitDegrees=2}
 {bone1Name=dyn_dress_right_b_04, bone2Name=dyn_dress_right_b_03, swing1LimitDegrees=9, swing2LimitDegrees=5, twistLimitDegrees=2}
 {bone1Name=dyn_dress_right_b_05, bone2Name=dyn_dress_right_b_04, swing1LimitDegrees=9, swing2LimitDegrees=5, twistLimitDegrees=2}
 {
  bone1Name=dyn_dress_left_b_01
  bone2Name=dyn_dress_left_b_master
  swing1LimitDegrees=15
  swing2LimitDegrees=5
  swing2Motion=Locked
  twistLimitDegrees=5
 }
 {
  bone1Name=dyn_dress_left_b_02
  bone2Name=dyn_dress_left_b_01
  swing1LimitDegrees=10
  swing2LimitDegrees=5
  swing2Motion=Locked
  twistLimitDegrees=5
 }
 {bone1Name=dyn_dress_left_b_03, bone2Name=dyn_dress_left_b_02, swing1LimitDegrees=10, swing2LimitDegrees=5, twistLimitDegrees=2}
 {bone1Name=dyn_dress_left_b_04, bone2Name=dyn_dress_left_b_03, swing1LimitDegrees=10, swing2LimitDegrees=5, twistLimitDegrees=2}
 {bone1Name=dyn_dress_left_b_05, bone2Name=dyn_dress_left_b_04, swing1LimitDegrees=10, swing2LimitDegrees=5, twistLimitDegrees=2}
 {
  bone1Name=dyn_dress_left_b_02
  bone2Name=dyn_dress_right_b_02
  swing1LimitDegrees=45
  swing1Motion=Free
  swing2LimitDegrees=45
  swing2Motion=Free
  twistLimitDegrees=45
  twistMotion=Free
 }
 {bone1Name=dyn_dress_right_a_01, bone2Name=dyn_dress_right_a_master, swing1LimitDegrees=15, swing2LimitDegrees=5, twistLimitDegrees=5}
 {bone1Name=dyn_dress_right_a_02, bone2Name=dyn_dress_right_a_01, swing1LimitDegrees=9, swing2LimitDegrees=5, twistLimitDegrees=5}
 {bone1Name=dyn_dress_right_a_03, bone2Name=dyn_dress_right_a_02, swing1LimitDegrees=9, swing2LimitDegrees=5, twistLimitDegrees=5}
 {bone1Name=dyn_dress_right_a_04, bone2Name=dyn_dress_right_a_03, swing1LimitDegrees=9, swing2LimitDegrees=5, twistLimitDegrees=2}
 {bone1Name=dyn_dress_right_a_05, bone2Name=dyn_dress_right_a_04, swing1LimitDegrees=9, swing2LimitDegrees=5, twistLimitDegrees=2}
 {bone1Name=dyn_dress_right_a_06, bone2Name=dyn_dress_right_a_05, swing1LimitDegrees=9, swing2LimitDegrees=5, twistLimitDegrees=2}
 {bone1Name=dyn_dress_left_a_01, bone2Name=dyn_dress_left_a_master, swing1LimitDegrees=10, swing2LimitDegrees=5, twistLimitDegrees=5}
 {bone1Name=dyn_dress_left_a_02, bone2Name=dyn_dress_left_a_01, swing1LimitDegrees=10, swing2LimitDegrees=5, twistLimitDegrees=5}
 {bone1Name=dyn_dress_left_a_03, bone2Name=dyn_dress_left_a_02, swing1LimitDegrees=10, swing2LimitDegrees=5, twistLimitDegrees=5}
 {bone1Name=dyn_dress_left_a_04, bone2Name=dyn_dress_left_a_03, swing1LimitDegrees=10, swing2LimitDegrees=5, twistLimitDegrees=2}
 {bone1Name=dyn_dress_left_a_05, bone2Name=dyn_dress_left_a_04, swing1LimitDegrees=10, swing2LimitDegrees=5, twistLimitDegrees=2}
 {bone1Name=dyn_dress_left_a_06, bone2Name=dyn_dress_left_a_05, swing1LimitDegrees=10, swing2LimitDegrees=5, twistLimitDegrees=2}
 {
  bone1Name=dyn_dress_right_a_06
  bone2Name=dyn_dress_right_b_05
  swing1LimitDegrees=45
  swing1Motion=Free
  swing2LimitDegrees=45
  swing2Motion=Free
  twistLimitDegrees=45
  twistMotion=Free
 }
 {
  bone1Name=dyn_dress_left_b_05
  bone2Name=dyn_dress_left_a_06
  swing1LimitDegrees=45
  swing1Motion=Free
  swing2LimitDegrees=45
  swing2Motion=Free
  twistLimitDegrees=45
  twistMotion=Free
 }
 {
  bone1Name=dyn_dress_right_b_01
  bone2Name=dyn_dress_left_b_01
  swing1LimitDegrees=45
  swing1Motion=Free
  swing2LimitDegrees=45
  swing2Motion=Free
  twistLimitDegrees=45
  twistMotion=Free
 }
 {
  bone1Name=dyn_dress_right_a_master
  bone2Name=root
  swing1LimitDegrees=45
  swing1Motion=Free
  swing2LimitDegrees=45
  swing2Motion=Free
  twistLimitDegrees=45
  twistMotion=Free
 }
 {
  bone1Name=dyn_dress_left_a_master
  bone2Name=root
  swing1LimitDegrees=45
  swing1Motion=Free
  swing2LimitDegrees=45
  swing2Motion=Free
  twistLimitDegrees=45
  twistMotion=Free
 }
 {
  bone1Name=dyn_dress_right_b_master
  bone2Name=root
  swing1LimitDegrees=45
  swing1Motion=Free
  swing2LimitDegrees=45
  swing2Motion=Free
  twistLimitDegrees=45
  twistMotion=Free
 }
 {
  bone1Name=dyn_dress_left_b_master
  bone2Name=root
  swing1LimitDegrees=45
  swing1Motion=Free
  swing2LimitDegrees=45
  swing2Motion=Free
  twistLimitDegrees=45
  twistMotion=Free
 }
 {
  bone1Name=thigh_l
  bone2Name=root
  swing1LimitDegrees=45
  swing1Motion=Free
  swing2LimitDegrees=45
  swing2Motion=Free
  twistLimitDegrees=45
  twistMotion=Free
 }
 {
  bone1Name=thigh_r
  bone2Name=root
  swing1LimitDegrees=45
  swing1Motion=Free
  swing2LimitDegrees=45
  swing2Motion=Free
  twistLimitDegrees=45
  twistMotion=Free
 }
 {
  bone1Name=calf_l
  bone2Name=thigh_l
  swing1LimitDegrees=45
  swing1Motion=Free
  swing2LimitDegrees=45
  swing2Motion=Free
  twistLimitDegrees=45
  twistMotion=Free
 }
 {
  bone1Name=calf_r
  bone2Name=thigh_r
  swing1LimitDegrees=45
  swing1Motion=Free
  swing2LimitDegrees=45
  swing2Motion=Free
  twistLimitDegrees=45
  twistMotion=Free
 }
]
