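# Physics asset for the crouch variant of the female medium Headband S adult body, covering its dynamic dress bones and leg collision.
# Root, thigh and calf shapes form the body, including a wide, thin box on the root.
# Four box-and-capsule dress chains hang from root-linked master bones with tight angular limits, with swing2 locked atop the b chains.
# Angularly free cross-links tie the upper b chains left to right and each side's a/b tips together.
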PhysicsAsset /BRCosmetics/Characters/Player/Female/Medium/Bodies/F_MED_Headband_S_Adult/Meshes/F_MED_Headband_S_Crouch_Adult_Physics
bodies {
 calf_l [{shapeType=Capsule, length=13.9296, radius=6.962, shapeName=calf_l_capsule, center=<-8.7422 0 0.0004>, rotation=<rot 90 0 -7.8952>}]
 calf_r [{shapeType=Capsule, length=14.6219, radius=7.121, shapeName=calf_r_capsule, center=<6.0055 0 0.0004>, rotation=<rot 90 0 -7.8951>}]
 dyn_dress_left_a_01 [{shapeType=Box, extentX=11.5, extentY=4, extentZ=11, shapeName=dyn_dress_left_a_01_box, center=<5.5933 0 0>, rotation=<rot 0 0 49.2836>}]
 dyn_dress_left_a_02 [{shapeType=Box, extentX=11.5, extentY=6, extentZ=9.5, shapeName=dyn_dress_left_a_02_box, center=<5.5933 0 0>, rotation=<rot 0 0 55.1246>}]
 dyn_dress_left_a_03 [{shapeType=Box, extentX=11.5, extentY=4, extentZ=11, shapeName=dyn_dress_left_a_03_box, center=<5.5933 0 0>, rotation=<rot 0 0 38.797>}]
 dyn_dress_left_a_04 [{shapeType=Capsule, length=6.25, radius=4.25, shapeName=dyn_dress_left_a_04_capsule, center=<7.5338 0 0>}]
 dyn_dress_left_a_05 [{shapeType=Capsule, length=4.75, radius=4.75, shapeName=dyn_dress_left_a_05_capsule, center=<3.7448 -0.0001 0>}]
 dyn_dress_left_a_06 [{shapeType=Capsule, length=6.25, radius=3.25, shapeName=dyn_dress_left_a_06_capsule, center=<4.2571 0 0>}]
 dyn_dress_left_a_master [{shapeType=Capsule, length=0.505, radius=0.505, shapeName=dyn_dress_left_a_master_capsule, rotation=<rot 0 -90 0>}]
 dyn_dress_left_b_01 [{shapeType=Box, extentX=11, extentY=3.5, extentZ=10, shapeName=dyn_dress_left_b_01_box, center=<5.4988 0.0002 0>}]
 dyn_dress_left_b_02 [{shapeType=Box, extentX=11, extentY=3.5, extentZ=10, shapeName=dyn_dress_left_b_02_box, center=<5.4988 0.0002 0>}]
 dyn_dress_left_b_03 [{shapeType=Capsule, length=6.25, radius=4.25, shapeName=dyn_dress_left_b_03_capsule, center=<7.5338 0 0>}]
 dyn_dress_left_b_04 [{shapeType=Capsule, length=4.75, radius=4.75, shapeName=dyn_dress_left_b_04_capsule, center=<3.7448 -0.0001 0>}]
 dyn_dress_left_b_05 [{shapeType=Capsule, length=6.25, radius=3.25, shapeName=dyn_dress_left_b_05_capsule, center=<4.2571 0 0>}]
 dyn_dress_left_b_master [{shapeType=Capsule, length=0.505, radius=0.505, shapeName=dyn_dress_left_b_master_capsule, rotation=<rot 0 -90 0>}]
 dyn_dress_right_a_01 [{shapeType=Box, extentX=11.5, extentY=6, extentZ=9, shapeName=dyn_dress_right_a_01_box, center=<5.5933 0 0>, rotation=<rot 0 0 -57.4517>}]
 dyn_dress_right_a_02 [{shapeType=Box, extentX=11.5, extentY=6, extentZ=9, shapeName=dyn_dress_right_a_02_box, center=<5.5933 0 0>, rotation=<rot 0 0 -57.4524>}]
 dyn_dress_right_a_03 [{shapeType=Box, extentX=11.5, extentY=6, extentZ=9, shapeName=dyn_dress_right_a_03_box, center=<5.5933 0 0>, rotation=<rot 0 0 -59.9293>}]
 dyn_dress_right_a_04 [{shapeType=Capsule, length=6.25, radius=4.25, shapeName=dyn_dress_right_a_04_capsule, center=<7.5338 1.7935 0>}]
 dyn_dress_right_a_05 [{shapeType=Capsule, length=4.75, radius=4.75, shapeName=dyn_dress_right_a_05_capsule, center=<3.7448 1.7934 0>}]
 dyn_dress_right_a_06 [{shapeType=Capsule, length=6.25, radius=3.25, shapeName=dyn_dress_right_a_06_capsule, center=<4.2571 1.7934 0>}]
 dyn_dress_right_a_master [{shapeType=Capsule, length=0.505, radius=0.505, shapeName=dyn_dress_right_a_master_capsule, rotation=<rot 0 -90 0>}]
 dyn_dress_right_b_01 [{shapeType=Box, extentX=11, extentY=5.5, extentZ=10, shapeName=dyn_dress_right_b_01_box, center=<5.4988 0.0002 0>}]
 dyn_dress_right_b_02 [{shapeType=Box, extentX=11, extentY=5.5, extentZ=10, shapeName=dyn_dress_right_b_02_box, center=<5.4988 0.0002 0>}]
 dyn_dress_right_b_03 [{shapeType=Capsule, length=6.25, radius=4.25, shapeName=dyn_dress_right_b_03_capsule, center=<7.5338 1.7935 0>}]
 dyn_dress_right_b_04 [{shapeType=Capsule, length=4.75, radius=4.75, shapeName=dyn_dress_right_b_04_capsule, center=<3.7448 1.7934 0>}]
 dyn_dress_right_b_05 [{shapeType=Capsule, length=6.25, radius=3.25, shapeName=dyn_dress_right_b_05_capsule, center=<4.2571 1.7934 0>}]
 dyn_dress_right_b_master [{shapeType=Capsule, length=0.505, radius=0.505, shapeName=dyn_dress_right_b_master_capsule, rotation=<rot 0 -90 0>}]
 root [
  {shapeType=Capsule, length=27, radius=19.75, shapeName=root_capsule, center=<0.5235 -11.8405 0>, rotation=<rot 90 0 -17.4032>}
  {shapeType=Box, extentX=100, extentY=100, extentZ=0.5, shapeName=root_box}
 ]
 thigh_l [{shapeType=Capsule, length=17.9268, radius=6.9143, shapeName=thigh_l_capsule, center=<-28.3503 1.5186 0>, rotation=<rot 90 0 0>}]
 thigh_r [{shapeType=Capsule, length=18.1468, radius=7.0349, shapeName=thigh_r_capsule, center=<26.3183 -1.938 -0.0001>, rotation=<rot 90 0 0>}]
}
constraints [
 {
  bone1Name=calf_r
  bone2Name=thigh_r
  swing1LimitDegrees=45
  swing1Motion=Free
  swing2LimitDegrees=45
  swing2Motion=Free
  twistLimitDegrees=45
  twistMotion=Free
 }
 {
  bone1Name=calf_l
  bone2Name=thigh_l
  swing1LimitDegrees=45
  swing1Motion=Free
  swing2LimitDegrees=45
  swing2Motion=Free
  twistLimitDegrees=45
  twistMotion=Free
 }
 {
  bone1Name=dyn_dress_right_b_01
  bone2Name=dyn_dress_right_b_master
  swing1LimitDegrees=15
  swing2LimitDegrees=5
  swing2Motion=Locked
  twistLimitDegrees=5
 }
 {
  bone1Name=dyn_dress_right_b_02
  bone2Name=dyn_dress_right_b_01
  swing1LimitDegrees=15
  swing2LimitDegrees=5
  swing2Motion=Locked
  twistLimitDegrees=5
 }
 {bone1Name=dyn_dress_right_b_03, bone2Name=dyn_dress_right_b_02, swing1LimitDegrees=20, swing2LimitDegrees=5, twistLimitDegrees=2}
 {bone1Name=dyn_dress_right_b_04, bone2Name=dyn_dress_right_b_03, swing1LimitDegrees=20, swing2LimitDegrees=5, twistLimitDegrees=2}
 {bone1Name=dyn_dress_right_b_05, bone2Name=dyn_dress_right_b_04, swing1LimitDegrees=20, swing2LimitDegrees=5, twistLimitDegrees=2}
 {
  bone1Name=dyn_dress_left_b_01
  bone2Name=dyn_dress_left_b_master
  swing1LimitDegrees=15
  swing2LimitDegrees=5
  swing2Motion=Locked
  twistLimitDegrees=5
 }
 {
  bone1Name=dyn_dress_left_b_02
  bone2Name=dyn_dress_left_b_01
  swing1LimitDegrees=15
  swing2LimitDegrees=5
  swing2Motion=Locked
  twistLimitDegrees=5
 }
 {bone1Name=dyn_dress_left_b_03, bone2Name=dyn_dress_left_b_02, swing1LimitDegrees=20, swing2LimitDegrees=5, twistLimitDegrees=2}
 {bone1Name=dyn_dress_left_b_04, bone2Name=dyn_dress_left_b_03, swing1LimitDegrees=20, swing2LimitDegrees=5, twistLimitDegrees=2}
 {bone1Name=dyn_dress_left_b_05, bone2Name=dyn_dress_left_b_04, swing1LimitDegrees=20, swing2LimitDegrees=5, twistLimitDegrees=2}
 {
  bone1Name=dyn_dress_left_b_02
  bone2Name=dyn_dress_right_b_02
  swing1LimitDegrees=45
  swing1Motion=Free
  swing2LimitDegrees=45
  swing2Motion=Free
  twistLimitDegrees=45
  twistMotion=Free
 }
 {bone1Name=dyn_dress_right_a_01, bone2Name=dyn_dress_right_a_master, swing1LimitDegrees=15, swing2LimitDegrees=5, twistLimitDegrees=5}
 {bone1Name=dyn_dress_right_a_02, bone2Name=dyn_dress_right_a_01, swing1LimitDegrees=15, swing2LimitDegrees=5, twistLimitDegrees=5}
 {bone1Name=dyn_dress_right_a_03, bone2Name=dyn_dress_right_a_02, swing1LimitDegrees=15, swing2LimitDegrees=5, twistLimitDegrees=5}
 {bone1Name=dyn_dress_right_a_04, bone2Name=dyn_dress_right_a_03, swing1LimitDegrees=20, swing2LimitDegrees=5, twistLimitDegrees=2}
 {bone1Name=dyn_dress_right_a_05, bone2Name=dyn_dress_right_a_04, swing1LimitDegrees=20, swing2LimitDegrees=5, twistLimitDegrees=2}
 {bone1Name=dyn_dress_right_a_06, bone2Name=dyn_dress_right_a_05, swing1LimitDegrees=20, swing2LimitDegrees=5, twistLimitDegrees=2}
 {bone1Name=dyn_dress_left_a_01, bone2Name=dyn_dress_left_a_master, swing1LimitDegrees=15, swing2LimitDegrees=5, twistLimitDegrees=5}
 {bone1Name=dyn_dress_left_a_02, bone2Name=dyn_dress_left_a_01, swing1LimitDegrees=15, swing2LimitDegrees=5, twistLimitDegrees=5}
 {bone1Name=dyn_dress_left_a_03, bone2Name=dyn_dress_left_a_02, swing1LimitDegrees=15, swing2LimitDegrees=5, twistLimitDegrees=5}
 {bone1Name=dyn_dress_left_a_04, bone2Name=dyn_dress_left_a_03, swing1LimitDegrees=20, swing2LimitDegrees=5, twistLimitDegrees=2}
 {bone1Name=dyn_dress_left_a_05, bone2Name=dyn_dress_left_a_04, swing1LimitDegrees=20, swing2LimitDegrees=5, twistLimitDegrees=2}
 {bone1Name=dyn_dress_left_a_06, bone2Name=dyn_dress_left_a_05, swing1LimitDegrees=20, swing2LimitDegrees=5, twistLimitDegrees=2}
 {
  bone1Name=dyn_dress_right_a_06
  bone2Name=dyn_dress_right_b_05
  swing1LimitDegrees=45
  swing1Motion=Free
  swing2LimitDegrees=45
  swing2Motion=Free
  twistLimitDegrees=45
  twistMotion=Free
 }
 {
  bone1Name=dyn_dress_left_b_05
  bone2Name=dyn_dress_left_a_06
  swing1LimitDegrees=45
  swing1Motion=Free
  swing2LimitDegrees=45
  swing2Motion=Free
  twistLimitDegrees=45
  twistMotion=Free
 }
 {
  bone1Name=dyn_dress_right_b_01
  bone2Name=dyn_dress_left_b_01
  swing1LimitDegrees=45
  swing1Motion=Free
  swing2LimitDegrees=45
  swing2Motion=Free
  twistLimitDegrees=45
  twistMotion=Free
 }
 {
  bone1Name=thigh_l
  bone2Name=root
  swing1LimitDegrees=45
  swing1Motion=Free
  swing2LimitDegrees=45
  swing2Motion=Free
  twistLimitDegrees=45
  twistMotion=Free
 }
 {
  bone1Name=thigh_r
  bone2Name=root
  swing1LimitDegrees=45
  swing1Motion=Free
  swing2LimitDegrees=45
  swing2Motion=Free
  twistLimitDegrees=45
  twistMotion=Free
 }
 {
  bone1Name=dyn_dress_right_a_master
  bone2Name=root
  swing1LimitDegrees=45
  swing1Motion=Free
  swing2LimitDegrees=45
  swing2Motion=Free
  twistLimitDegrees=45
  twistMotion=Free
 }
 {
  bone1Name=dyn_dress_left_a_master
  bone2Name=root
  swing1LimitDegrees=45
  swing1Motion=Free
  swing2LimitDegrees=45
  swing2Motion=Free
  twistLimitDegrees=45
  twistMotion=Free
 }
 {
  bone1Name=dyn_dress_right_b_master
  bone2Name=root
  swing1LimitDegrees=45
  swing1Motion=Free
  swing2LimitDegrees=45
  swing2Motion=Free
  twistLimitDegrees=45
  twistMotion=Free
 }
 {
  bone1Name=dyn_dress_left_b_master
  bone2Name=root
  swing1LimitDegrees=45
  swing1Motion=Free
  swing2LimitDegrees=45
  swing2Motion=Free
  twistLimitDegrees=45
  twistMotion=Free
 }
]
