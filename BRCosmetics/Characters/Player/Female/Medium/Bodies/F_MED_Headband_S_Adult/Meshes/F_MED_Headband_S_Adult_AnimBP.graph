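# Post-copy AnimBP for the female Headband outfit body, animating its dress bones on top of the parent mesh's pose.
# Skydive and ground branches each run Control Rig, Look At aiming and pelvis-space RigidBody, blended by skydive state, with an extra front-end simulation.
# Additive bone offsets then correct the dress, weighted by thigh and pelvis deltas from ref pose that the EventGraph remaps per crouch state, plus front-end tweaks.
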AnimBlueprint /BRCosmetics/Characters/Player/Female/Medium/Bodies/F_MED_Headband_S_Adult/Meshes/F_MED_Headband_S_Adult_AnimBP
target_skeleton=&/F_MED_Headband_S_Adult_Skeleton
graphs [
 AnimGraph {nodes [
   "Transform (Modify) Bone - Bone: dyn_dress_left_a_01_deform"(Alpha=1, Translation="0,0,0", Rotation="0,0,0", Scale="1,1,1") {
    rotationMode=BMM_Additive
    rotationSpace=BCS_BoneSpace
    translationMode=BMM_Additive
    translationSpace=BCS_BoneSpace
    alphaBoolBlend {blendInTime=0.25, blendOption=Linear, blendOutTime=0.25}
    alphaScaleBiasClamp {bInterpResult=true}
    boneToModify {boneName=dyn_dress_left_a_01_deform}
   }
   n1 = "Transform (Modify) Bone - Bone: dyn_dress_right_b_05"!(ComponentPose, bAlphaBoolEnabled:bool=True, Translation="0,0,0", Rotation="-0.012,9.067,-0.008", Scale="1,1,1") -> (Pose) {
    alphaInputType=Bool
    rotationMode=BMM_Additive
    rotationSpace=BCS_BoneSpace
    translationMode=BMM_Additive
    translationSpace=BCS_BoneSpace
    alphaBoolBlend {blendInTime=0.25, blendOption=Linear, blendOutTime=0.25}
    alphaScaleBiasClamp {bInterpResult=true}
    boneToModify {boneName=dyn_dress_right_b_05}
    rotation=<rot -0.012 9.067 -0.008>
   }
   n2 = "Transform (Modify) Bone - Bone: dyn_dress_left_b_05"!(ComponentPose=@n1, bAlphaBoolEnabled:bool=True, Translation="0,0,0", Rotation="0,-15,0", Scale="1,1,1") -> (Pose) {
    alphaInputType=Bool
    rotationMode=BMM_Additive
    rotationSpace=BCS_BoneSpace
    translationMode=BMM_Additive
    translationSpace=BCS_BoneSpace
    alphaBoolBlend {blendInTime=0.25, blendOption=Linear, blendOutTime=0.25}
    alphaScaleBiasClamp {bInterpResult=true}
    boneToModify {boneName=dyn_dress_left_b_05}
    rotation {yaw=-15}
   }
   n3 = "Transform (Modify) Bone - Bone: dyn_dress_right_a_06"!(ComponentPose=@n2, bAlphaBoolEnabled:bool=True, Translation="0,0,0", Rotation="0,-15,0", Scale="1,1,1") -> (Pose) {
    alphaInputType=Bool
    rotationMode=BMM_Additive
    rotationSpace=BCS_BoneSpace
    translationMode=BMM_Additive
    translationSpace=BCS_BoneSpace
    alphaBoolBlend {blendInTime=0.25, blendOption=Linear, blendOutTime=0.25}
    alphaScaleBiasClamp {bInterpResult=true}
    boneToModify {boneName=dyn_dress_right_a_06}
    rotation {yaw=-15}
   }
   "Transform (Modify) Bone - Bone: dyn_dress_left_a_06"!(ComponentPose=@n3, bAlphaBoolEnabled:bool=True, Translation="0,0,0", Rotation="0,-15,0", Scale="1,1,1") -> (Pose) {
    alphaInputType=Bool
    rotationMode=BMM_Additive
    rotationSpace=BCS_BoneSpace
    translationMode=BMM_Additive
    translationSpace=BCS_BoneSpace
    alphaBoolBlend {blendInTime=0.25, blendOption=Linear, blendOutTime=0.25}
    alphaScaleBiasClamp {bInterpResult=true}
    boneToModify {boneName=dyn_dress_left_a_06}
    rotation {yaw=-15}
   }
   n4 = Control Rig(Source=Copy Pose From Mesh() {bUseAttachedParent=true}) {
    alphaBoolBlend {blendOption=Linear}
    controlRigAssetReference {blueprintRigClass=&/F_MED_Headband_S_Adult_CtrlRig.F_MED_Headband_S_Adult_CtrlRig_C}
   }
   n5 = "Transform (Modify) Bone - Bone: dress_aim_l"(ComponentPose=Local To Component(LocalPose=@n4), Alpha=1, Translation="13.778,-37.430,44.404", Rotation="0,0,0", Scale="1,1,1") {
    translationMode=BMM_Additive
    translationSpace=BCS_BoneSpace
    alphaBoolBlend {blendOption=Linear}
    boneToModify {boneName=dress_aim_l}
    translation=<13.778 -37.43 44.404>
   }
   n6 = Get bIsCrouching()
   n7 = Look At - Bone: dyn_dress_right_a_master!(ComponentPose=@n5, bAlphaBoolEnabled:bool=@n6) -> (Pose) {
    alphaInputType=Bool
    alphaBoolBlend {blendOption=Linear}
    boneToModify {boneName=dyn_dress_right_a_master}
    lookAtLocation {x=-25}
    lookAtTarget {boneReference {boneName=dress_aim_l}}
    lookAt_Axis {axis=<1 0>}
   }
   n8 = Look At - Bone: dyn_dress_left_a_master!(ComponentPose=@n7, bAlphaBoolEnabled:bool=@n6) -> (Pose) {
    alphaInputType=Bool
    alphaBoolBlend {blendOption=Linear}
    boneToModify {boneName=dyn_dress_left_a_master}
    lookAtLocation {x=25}
    lookAtTarget {boneReference {boneName=dress_aim_l}}
    lookAt_Axis {axis=<1 0>}
   }
   n9 = Look At - Bone: dyn_dress_right_b_master!(ComponentPose=@n8, bAlphaBoolEnabled:bool=@n6) -> (Pose) {
    alphaInputType=Bool
    alphaBoolBlend {blendOption=Linear}
    boneToModify {boneName=dyn_dress_right_b_master}
    lookAtLocation {x=-12}
    lookAtTarget {boneReference {boneName=dress_aim_l}}
    lookAt_Axis {axis=<1 0>}
   }
   n10 = Look At - Bone: dyn_dress_left_b_master!(ComponentPose=@n9, bAlphaBoolEnabled:bool=@n6) -> (Pose) {
    alphaInputType=Bool
    alphaBoolBlend {blendOption=Linear}
    boneToModify {boneName=dyn_dress_left_b_master}
    lookAtLocation {x=12}
    lookAtTarget {boneReference {boneName=dress_aim_l}}
    lookAt_Axis {axis=<1 0>}
   }
   n11 = Get GravityOverride()
   n12 = RigidBody(ComponentPose=@n10, ExternalForce=@n11) {
    bOverrideWorldGravity=true
    overridePhysicsAsset=&/F_MED_Headband_S_Crouch_Adult_Physics
    simulationSpace=BaseBoneSpace
    alphaBoolBlend {blendOption=Linear}
    baseBoneRef {boneName=pelvis}
    componentAppliedLinearAccClamp=<2000 2000 2000>
    componentLinearAccScale=<0.3 0.3 0.35>
    componentLinearVelScale=<0.3 0.3 0.35>
    overrideWorldGravity {z=-980}
   }
   n13 = RigidBody!(ComponentPose=@n12, bAlphaBoolEnabled:bool=Get bIsInFrontEnd(), ExternalForce=@n11) -> (Pose) {
    alphaInputType=Bool
    bOverrideWorldGravity=true
    overridePhysicsAsset=&/F_MED_Headband_S_Adult_FrontEnd_Physics
    simulationSpace=BaseBoneSpace
    alphaBoolBlend {blendOption=Linear}
    baseBoneRef {boneName=pelvis}
    componentAppliedLinearAccClamp=<2000 2000 2000>
    componentLinearAccScale=<0.3 0.3 0.35>
    componentLinearVelScale=<0.3 0.3 0.35>
    overrideWorldGravity {z=-980}
   }
   n14 = Component To Local(ComponentPose=@n13)
   n15 = Control Rig(Source=Copy Pose From Mesh() {bUseAttachedParent=true}) {
    alphaBoolBlend {blendOption=Linear}
    controlRigAssetReference {blueprintRigClass=&F_MED_Headband_S_Adult_CtrlRig_C}
   }
   n16 = "Transform (Modify) Bone - Bone: dress_aim_l"(ComponentPose=Local To Component(LocalPose=@n15), Alpha=1, Translation="0,-10,60", Rotation="0,0,0", Scale="1,1,1") {
    translationMode=BMM_Additive
    translationSpace=BCS_BoneSpace
    alphaBoolBlend {blendOption=Linear}
    boneToModify {boneName=dress_aim_l}
    translation {y=-10, z=60}
   }
   n17 = "Transform (Modify) Bone - Bone: dress_aim_l"!(ComponentPose=@n16, bAlphaBoolEnabled:bool=Get bIsSkydiveDiveMode(), Translation="0,20,0", Rotation="0,0,0", Scale="1,1,1") -> (Pose) {
    alphaInputType=Bool
    translationMode=BMM_Additive
    translationSpace=BCS_BoneSpace
    alphaBoolBlend {blendInTime=2, blendOption=Linear, blendOutTime=2}
    boneToModify {boneName=dress_aim_l}
    translation {y=20}
   }
   n18 = Get Skydiving() -> (Skydiving:bool)
   n19 = Look At - Bone: dyn_dress_right_a_master(ComponentPose=@n17, bAlphaBoolEnabled=@n18) {
    alphaInputType=Bool
    alphaBoolBlend {blendOption=Linear}
    boneToModify {boneName=dyn_dress_right_a_master}
    lookAtLocation {x=-25}
    lookAtTarget {boneReference {boneName=dress_aim_l}}
    lookAt_Axis {axis=<1 0>}
   }
   n20 = Look At - Bone: dyn_dress_left_a_master(ComponentPose=@n19, bAlphaBoolEnabled=@n18) {
    alphaInputType=Bool
    alphaBoolBlend {blendOption=Linear}
    boneToModify {boneName=dyn_dress_left_a_master}
    lookAtLocation {x=25}
    lookAtTarget {boneReference {boneName=dress_aim_l}}
    lookAt_Axis {axis=<1 0>}
   }
   n21 = Look At - Bone: dyn_dress_right_b_master(ComponentPose=@n20, bAlphaBoolEnabled=@n18) {
    alphaInputType=Bool
    alphaBoolBlend {blendOption=Linear}
    boneToModify {boneName=dyn_dress_right_b_master}
    lookAtLocation {x=-12}
    lookAtTarget {boneReference {boneName=dress_aim_l}}
    lookAt_Axis {axis=<1 0>}
   }
   n22 = Look At - Bone: dyn_dress_left_b_master(ComponentPose=@n21, bAlphaBoolEnabled=@n18) {
    alphaInputType=Bool
    alphaBoolBlend {blendOption=Linear}
    boneToModify {boneName=dyn_dress_left_b_master}
    lookAtLocation {x=12}
    lookAtTarget {boneReference {boneName=dress_aim_l}}
    lookAt_Axis {axis=<1 0>}
   }
   n23 = RigidBody!(ComponentPose=@n22, Alpha:float, ExternalForce=Get GravityOverride()) -> (Pose) {
    bOverrideWorldGravity=true
    overridePhysicsAsset=&/F_MED_Headband_S_Adult_Skydive_PHYS
    simulationSpace=BaseBoneSpace
    alphaBoolBlend {blendOption=Linear}
    baseBoneRef {boneName=pelvis}
    componentAppliedLinearAccClamp=<2000 2000 2000>
    componentLinearAccScale=<0.3 0.3 0.35>
    componentLinearVelScale=<0.3 0.3 0.35>
    overrideWorldGravity {z=-980}
   }
   n24 = Get bIsSkydiveDiveMode()
   n25 = "Transform (Modify) Bone - Bone: dyn_dress_right_a_02"!(ComponentPose=@n23, bAlphaBoolEnabled:bool=@n24, Translation="0,0,0", Rotation="0,30,0", Scale="1,1,1") -> (Pose) {
    alphaInputType=Bool
    rotationMode=BMM_Additive
    rotationSpace=BCS_BoneSpace
    translationMode=BMM_Additive
    translationSpace=BCS_BoneSpace
    alphaBoolBlend {blendInTime=1, blendOption=Linear, blendOutTime=1}
    boneToModify {boneName=dyn_dress_right_a_02}
    rotation {yaw=30}
   }
   n26 = "Transform (Modify) Bone - Bone: dyn_dress_right_a_03"!(ComponentPose=@n25, bAlphaBoolEnabled:bool=@n24, Translation="0,0,0", Rotation="0,30,0", Scale="1,1,1") -> (Pose) {
    alphaInputType=Bool
    rotationMode=BMM_Additive
    rotationSpace=BCS_BoneSpace
    translationMode=BMM_Additive
    translationSpace=BCS_BoneSpace
    alphaBoolBlend {blendInTime=1, blendOption=Linear, blendOutTime=1}
    boneToModify {boneName=dyn_dress_right_a_03}
    rotation {yaw=30}
   }
   n27 = "Transform (Modify) Bone - Bone: dyn_dress_right_a_04"!(ComponentPose=@n26, bAlphaBoolEnabled:bool=@n24, Translation="0,0,0", Rotation="0,30,0", Scale="1,1,1") -> (Pose) {
    alphaInputType=Bool
    rotationMode=BMM_Additive
    rotationSpace=BCS_BoneSpace
    translationMode=BMM_Additive
    translationSpace=BCS_BoneSpace
    alphaBoolBlend {blendInTime=1, blendOption=Linear, blendOutTime=1}
    boneToModify {boneName=dyn_dress_right_a_04}
    rotation {yaw=30}
   }
   n28 = "Transform (Modify) Bone - Bone: dyn_dress_right_b_03"!(ComponentPose=@n27, bAlphaBoolEnabled:bool=@n24, Translation="0,0,0", Rotation="0,30,0", Scale="1,1,1") -> (Pose) {
    alphaInputType=Bool
    rotationMode=BMM_Additive
    rotationSpace=BCS_BoneSpace
    translationMode=BMM_Additive
    translationSpace=BCS_BoneSpace
    alphaBoolBlend {blendInTime=1, blendOption=Linear, blendOutTime=1}
    boneToModify {boneName=dyn_dress_right_b_03}
    rotation {yaw=30}
   }
   n29 = "Transform (Modify) Bone - Bone: dyn_dress_right_b_02"!(ComponentPose=@n28, bAlphaBoolEnabled:bool=@n24, Translation="0,0,0", Rotation="0,30,0", Scale="1,1,1") -> (Pose) {
    alphaInputType=Bool
    rotationMode=BMM_Additive
    rotationSpace=BCS_BoneSpace
    translationMode=BMM_Additive
    translationSpace=BCS_BoneSpace
    alphaBoolBlend {blendInTime=1, blendOption=Linear, blendOutTime=1}
    boneToModify {boneName=dyn_dress_right_b_02}
    rotation {yaw=30}
   }
   n30 = Component To Local(ComponentPose=@n29)
   n31 = "Transform (Modify) Bone - Bone: dyn_dress_right_a_master"!(
    ComponentPose=Local To Component(LocalPose=Blend Poses by bool(BlendPose_0=@n30, BlendPose_1=@n14, bActiveValue=@n18) {blendType=Linear})
    bAlphaBoolEnabled:bool=True
    Translation="-0.000,-0.000,-5.106"
    Rotation="0,0,0"
    Scale="1,1,1"
   ) -> (Pose) {
    alphaInputType=Bool
    rotationMode=BMM_Additive
    rotationSpace=BCS_BoneSpace
    translationMode=BMM_Additive
    translationSpace=BCS_BoneSpace
    alphaBoolBlend {blendInTime=0.25, blendOption=Linear, blendOutTime=0.25}
    boneToModify {boneName=dyn_dress_right_a_master}
    translation {z=-5.106}
   }
   n32 = "Transform (Modify) Bone - Bone: dyn_dress_right_a_02"(ComponentPose=@n31, bAlphaBoolEnabled=True, Translation="-0.000,-0.023,4.214", Rotation="0,0,0", Scale="1,1,1") {
    alphaInputType=Bool
    rotationMode=BMM_Additive
    rotationSpace=BCS_BoneSpace
    translationMode=BMM_Additive
    translationSpace=BCS_BoneSpace
    alphaBoolBlend {blendInTime=0.25, blendOption=Linear, blendOutTime=0.25}
    boneToModify {boneName=dyn_dress_right_a_02}
    translation {y=-0.023, z=4.214}
   }
   n33 = "Transform (Modify) Bone - Bone: dyn_dress_right_a_02"(ComponentPose=@n32, bAlphaBoolEnabled=True, Translation="-0.000,-0.027,0.5", Rotation="0,0,0", Scale="1,1,1") {
    alphaInputType=Bool
    rotationMode=BMM_Additive
    rotationSpace=BCS_BoneSpace
    translationMode=BMM_Additive
    translationSpace=BCS_BoneSpace
    alphaBoolBlend {blendInTime=0.25, blendOption=Linear, blendOutTime=0.25}
    boneToModify {boneName=dyn_dress_right_a_02}
    translation {y=-0.027, z=0.5}
   }
   n34 = "Transform (Modify) Bone - Bone: dyn_dress_left_a_master"!(ComponentPose=@n33, bAlphaBoolEnabled:bool=True, Translation="-0.000,-0.000,2.649", Rotation="0,0,0", Scale="1,1,1") -> (Pose) {
    alphaInputType=Bool
    rotationMode=BMM_Additive
    rotationSpace=BCS_BoneSpace
    translationMode=BMM_Additive
    translationSpace=BCS_BoneSpace
    alphaBoolBlend {blendInTime=0.25, blendOption=Linear, blendOutTime=0.25}
    boneToModify {boneName=dyn_dress_left_a_master}
    translation {z=2.649}
   }
   n35 = "Transform (Modify) Bone - Bone: dyn_dress_left_a_02"!(ComponentPose=@n34, bAlphaBoolEnabled:bool=True, Translation="-0.000,0.014,0.144", Rotation="0,0,0", Scale="1,1,1") -> (Pose) {
    alphaInputType=Bool
    rotationMode=BMM_Additive
    rotationSpace=BCS_BoneSpace
    translationMode=BMM_Additive
    translationSpace=BCS_BoneSpace
    alphaBoolBlend {blendInTime=0.25, blendOption=Linear, blendOutTime=0.25}
    boneToModify {boneName=dyn_dress_left_a_02}
    translation {y=0.014, z=0.144}
   }
   n36 = Get bIsCrouching()
   n37 = "Transform (Modify) Bone - Bone: dyn_dress_right_a_01_deform"!(ComponentPose=@n35, bAlphaBoolEnabled:bool=@n36, Translation="-5.283,-0.124,-3.359", Rotation="0,0,0", Scale="1,1,1") -> (Pose) {
    alphaInputType=Bool
    rotationMode=BMM_Additive
    rotationSpace=BCS_BoneSpace
    translationMode=BMM_Additive
    translationSpace=BCS_BoneSpace
    alphaBoolBlend {blendInTime=0.25, blendOption=Linear, blendOutTime=0.25}
    alphaScaleBiasClamp {bInterpResult=true}
    boneToModify {boneName=dyn_dress_right_a_01_deform}
    translation=<-5.283 -0.124 -3.359>
   }
   n38 = "Transform (Modify) Bone - Bone: dyn_dress_right_a_03"(ComponentPose=@n37, bAlphaBoolEnabled=@n36, Translation="-0.681,1.829,2", Rotation="0,9.999,0.005", Scale="1,1,1") {
    alphaInputType=Bool
    rotationMode=BMM_Additive
    rotationSpace=BCS_BoneSpace
    translationMode=BMM_Additive
    translationSpace=BCS_BoneSpace
    alphaBoolBlend {blendInTime=0.5, blendOption=Linear, blendOutTime=0.5}
    alphaScaleBiasClamp {bInterpResult=true}
    boneToModify {boneName=dyn_dress_right_a_03}
    rotation {roll=0.005, yaw=9.999}
    translation=<-0.681 1.829 2>
   }
   n39 = "Transform (Modify) Bone - Bone: dyn_dress_right_b_02"(ComponentPose=@n38, bAlphaBoolEnabled=@n36, Translation="-0.321,-0.041,0", Rotation="7.188,10.000,0.000", Scale="1,1,1") {
    alphaInputType=Bool
    rotationMode=BMM_Additive
    rotationSpace=BCS_BoneSpace
    translationMode=BMM_Additive
    translationSpace=BCS_BoneSpace
    alphaBoolBlend {blendInTime=0.5, blendOption=Linear, blendOutTime=0.5}
    alphaScaleBiasClamp {bInterpResult=true}
    boneToModify {boneName=dyn_dress_right_b_02}
    rotation {pitch=7.188, yaw=10}
    translation=<-0.321 -0.041>
   }
   n40 = "Transform (Modify) Bone - Bone: dyn_dress_right_a_01_deform"!(
    ComponentPose=@n39
    Alpha:float=Get Thigh R Crouch Rot Alpha() -> (Thigh R Crouch Rot Alpha:double)
    Translation="-20.218,0.240,0"
    Rotation="0,0,0"
    Scale="1,1,1"
   ) -> (Pose) {
    rotationMode=BMM_Additive
    rotationSpace=BCS_BoneSpace
    translationMode=BMM_Additive
    translationSpace=BCS_BoneSpace
    alphaBoolBlend {blendInTime=0.25, blendOption=Linear, blendOutTime=0.25}
    boneToModify {boneName=dyn_dress_right_a_01_deform}
    translation=<-20.218 0.24>
   }
   n41 = "Transform (Modify) Bone - Bone: dyn_dress_right_a_01_deform"(
    ComponentPose=@n40
    Alpha=Get Thigh R Crouch Pitch Alpha() -> (Thigh R Crouch Pitch Alpha:double)
    Translation="-10.991,-6.412,11.206"
    Rotation="0,0,0"
    Scale="1,1,1"
   ) {
    rotationMode=BMM_Additive
    rotationSpace=BCS_BoneSpace
    translationMode=BMM_Additive
    alphaBoolBlend {blendInTime=0.25, blendOption=Linear, blendOutTime=0.25}
    alphaScaleBiasClamp {bInterpResult=true, interpSpeedIncreasing=15}
    boneToModify {boneName=dyn_dress_right_a_01_deform}
    translation=<-10.991 -6.412 11.206>
   }
   n42 = "Transform (Modify) Bone - Bone: dyn_dress_left_a_01_deform"(
    ComponentPose=@n41
    Alpha=Get Thigh L Crouch Rot Alpha() -> (Thigh L Crouch Rot Alpha:double)
    Translation="-15.770,-15.802,-3.885"
    Rotation="-16.788,-24.999,-0.002"
    Scale="1,1,1"
   ) {
    rotationMode=BMM_Additive
    rotationSpace=BCS_BoneSpace
    translationMode=BMM_Additive
    translationSpace=BCS_BoneSpace
    alphaBoolBlend {blendInTime=0.25, blendOption=Linear, blendOutTime=0.25}
    alphaScaleBiasClamp {bInterpResult=true, interpSpeedIncreasing=15}
    boneToModify {boneName=dyn_dress_left_a_01_deform}
    rotation=<rot -16.788 -24.9987 -0.002>
    translation=<-15.77 -15.802 -3.885>
   }
   n43 = "Transform (Modify) Bone - Bone: dyn_dress_left_a_01_deform"(ComponentPose=@n42, Alpha=Get Pelvis Rot() -> (Pelvis Rot:double), Translation="-0.000,0.007,6", Rotation="0,0,0", Scale="1,1,1") {
    rotationMode=BMM_Additive
    rotationSpace=BCS_BoneSpace
    translationMode=BMM_Additive
    translationSpace=BCS_BoneSpace
    alphaBoolBlend {blendInTime=0.25, blendOption=Linear, blendOutTime=0.25}
    alphaScaleBiasClamp {bInterpResult=true}
    boneToModify {boneName=dyn_dress_left_a_01_deform}
    translation {y=0.007, z=6}
   }
   n44 = "Transform (Modify) Bone - Bone: dyn_dress_left_a_01_deform"(
    ComponentPose=@n43
    Alpha=Get Thigh L Crouch Pitch Alpha() -> (Thigh L Crouch Pitch Alpha:double)
    Translation="-8.470,0.048,15.610"
    Rotation="0.000,0.000,-24.043"
    Scale="1,1,1"
   ) {
    rotationMode=BMM_Additive
    rotationSpace=BCS_BoneSpace
    translationMode=BMM_Additive
    translationSpace=BCS_BoneSpace
    alphaBoolBlend {blendInTime=0.25, blendOption=Linear, blendOutTime=0.25}
    alphaScaleBiasClamp {bInterpResult=true}
    boneToModify {boneName=dyn_dress_left_a_01_deform}
    rotation {roll=-24.0428}
    translation=<-8.47 0.048 15.61>
   }
   n45 = "Transform (Modify) Bone - Bone: dyn_dress_right_a_01_deform"(ComponentPose=@n44, Alpha=Get Thigh R Crouch Rot Alpha(), Translation="5.003,0.000,11.893", Rotation="0.000,7.482,0.000", Scale="1,1,1") {
    rotationMode=BMM_Additive
    translationMode=BMM_Additive
    alphaBoolBlend {blendInTime=0.25, blendOption=Linear, blendOutTime=0.25}
    alphaScaleBiasClamp {bInterpResult=true, interpSpeedIncreasing=15}
    boneToModify {boneName=dyn_dress_right_a_01_deform}
    rotation {yaw=7.482}
    translation {x=5.003, z=11.893}
   }
   n46 = "Transform (Modify) Bone - Bone: dyn_dress_right_a_01_deform"(
    ComponentPose=@n45
    Alpha=Get Thigh R Pitch Alpha() -> (Thigh R Pitch Alpha:double)
    Translation="-0.016,-0.046,-8.926"
    Rotation="-14.772,-0.003,-0.043"
    Scale="1,1,1"
   ) {
    rotationMode=BMM_Additive
    rotationSpace=BCS_BoneSpace
    translationMode=BMM_Additive
    translationSpace=BCS_BoneSpace
    alphaBoolBlend {blendInTime=0.25, blendOption=Linear, blendOutTime=0.25}
    alphaScaleBiasClamp {bInterpResult=true}
    boneToModify {boneName=dyn_dress_right_a_01_deform}
    rotation=<rot -14.772 -0.003 -0.043>
    translation=<-0.016 -0.046 -8.926>
   }
   n47 = "Transform (Modify) Bone - Bone: dyn_dress_left_a_01_deform"(
    ComponentPose=@n46
    Alpha=Get Thigh L Pitch Alpha() -> (Thigh L Pitch Alpha:double)
    Translation="-0.000,0.079,6.506"
    Rotation="0,0,0"
    Scale="1,1,1"
   ) {
    rotationMode=BMM_Additive
    rotationSpace=BCS_BoneSpace
    translationMode=BMM_Additive
    translationSpace=BCS_BoneSpace
    alphaBoolBlend {blendInTime=0.25, blendOption=Linear, blendOutTime=0.25}
    alphaScaleBiasClamp {bInterpResult=true, interpSpeedIncreasing=15}
    boneToModify {boneName=dyn_dress_left_a_01_deform}
    translation {y=0.079, z=6.506}
   }
   n48 = "Transform (Modify) Bone - Bone: dyn_dress_right_a_01_deform"(
    ComponentPose=@n47
    Alpha=Get Thigh R Rot Alpha() -> (Thigh R Rot Alpha:double)
    Translation="-18.705,-2.430,-7.864"
    Rotation="0,0,0"
    Scale="1,1,1"
   ) {
    rotationMode=BMM_Additive
    rotationSpace=BCS_BoneSpace
    translationMode=BMM_Additive
    translationSpace=BCS_BoneSpace
    alphaBoolBlend {blendInTime=0.25, blendOption=Linear, blendOutTime=0.25}
    alphaScaleBiasClamp {bInterpResult=true}
    boneToModify {boneName=dyn_dress_right_a_01_deform}
    translation=<-18.705 -2.43 -7.864>
   }
   n49 = "Transform (Modify) Bone - Bone: dyn_dress_left_a_01_deform"(
    ComponentPose=@n48
    Alpha=Get Thigh L Rot Alpha() -> (Thigh L Rot Alpha:double)
    Translation="-11.531,0.052,4.002"
    Rotation="0,0,0"
    Scale="1,1,1"
   ) {
    rotationMode=BMM_Additive
    rotationSpace=BCS_BoneSpace
    translationMode=BMM_Additive
    translationSpace=BCS_BoneSpace
    alphaBoolBlend {blendInTime=0.25, blendOption=Linear, blendOutTime=0.25}
    alphaScaleBiasClamp {bInterpResult=true, interpSpeedIncreasing=15}
    boneToModify {boneName=dyn_dress_left_a_01_deform}
    translation=<-11.531 0.052 4.002>
   }
   n50 = Get Skydiving()
   n51 = Look At - Bone: dyn_dress_right_a_03!(ComponentPose=@n49, bAlphaBoolEnabled:bool=@n50) -> (Pose) {
    alphaInputType=Bool
    alphaBoolBlend {blendOption=Linear}
    boneToModify {boneName=dyn_dress_right_a_03}
    lookAtLocation {x=-25}
    lookAtTarget {boneReference {boneName=dress_aim_l}}
    lookAt_Axis {axis=<1 0>}
   }
   n52 = Look At - Bone: dyn_dress_left_a_03!(ComponentPose=@n51, bAlphaBoolEnabled:bool=@n50) -> (Pose) {
    alphaInputType=Bool
    alphaBoolBlend {blendOption=Linear}
    boneToModify {boneName=dyn_dress_left_a_03}
    lookAtLocation {x=25}
    lookAtTarget {boneReference {boneName=dress_aim_l}}
    lookAt_Axis {axis=<1 0>}
   }
   n53 = Look At - Bone: dyn_dress_right_b_02!(ComponentPose=@n52, bAlphaBoolEnabled:bool=@n50) -> (Pose) {
    alphaInputType=Bool
    alphaBoolBlend {blendOption=Linear}
    boneToModify {boneName=dyn_dress_right_b_02}
    lookAtLocation {x=-12}
    lookAtTarget {boneReference {boneName=dress_aim_l}}
    lookAt_Axis {axis=<1 0>}
   }
   n54 = Look At - Bone: dyn_dress_left_b_02!(ComponentPose=@n53, bAlphaBoolEnabled:bool=@n50) -> (Pose) {
    alphaInputType=Bool
    alphaBoolBlend {blendOption=Linear}
    boneToModify {boneName=dyn_dress_left_b_02}
    lookAtLocation {x=12}
    lookAtTarget {boneReference {boneName=dress_aim_l}}
    lookAt_Axis {axis=<1 0>}
   }
   n55 = Get FrontendNotEmote() -> (FrontendNotEmote:bool)
   n56 = "Transform (Modify) Bone - Bone: dyn_dress_left_a_01_deform"!(ComponentPose=@n54, bAlphaBoolEnabled:bool=@n55, Translation="0,0,-5", Rotation="0,0,0", Scale="1,1,1") -> (Pose) {
    alphaInputType=Bool
    rotationMode=BMM_Additive
    rotationSpace=BCS_BoneSpace
    translationMode=BMM_Additive
    translationSpace=BCS_BoneSpace
    alphaBoolBlend {blendInTime=0.25, blendOption=Linear, blendOutTime=0.25}
    alphaScaleBiasClamp {bInterpResult=true}
    boneToModify {boneName=dyn_dress_left_a_01_deform}
    translation {z=-5}
   }
   n57 = "Transform (Modify) Bone - Bone: dyn_dress_right_a_01_deform"!(ComponentPose=@n56, bAlphaBoolEnabled:bool=@n55, Translation="0,0,5", Rotation="0,0,0", Scale="1,1,1") -> (Pose) {
    alphaInputType=Bool
    rotationMode=BMM_Additive
    rotationSpace=BCS_BoneSpace
    translationMode=BMM_Additive
    translationSpace=BCS_BoneSpace
    alphaBoolBlend {blendInTime=0.25, blendOption=Linear, blendOutTime=0.25}
    alphaScaleBiasClamp {bInterpResult=true}
    boneToModify {boneName=dyn_dress_right_a_01_deform}
    translation {z=5}
   }
   n58 = Get FrontendNotEmoteRbanDisabled() -> (FrontendNotEmoteRbanDisabled:bool)
   n59 = "Transform (Modify) Bone - Bone: dyn_dress_left_a_master"(ComponentPose=@n57, bAlphaBoolEnabled=@n58, Translation="0,0,0", Rotation="30,-10,0", Scale="1,1,1") {
    alphaInputType=Bool
    rotationMode=BMM_Additive
    rotationSpace=BCS_BoneSpace
    translationMode=BMM_Additive
    translationSpace=BCS_BoneSpace
    alphaBoolBlend {blendInTime=0.25, blendOption=Linear, blendOutTime=0.25}
    alphaScaleBiasClamp {bInterpResult=true}
    boneToModify {boneName=dyn_dress_left_a_master}
    rotation {pitch=30, yaw=-10}
   }
   n60 = "Transform (Modify) Bone - Bone: dyn_dress_left_b_master"!(ComponentPose=@n59, bAlphaBoolEnabled:bool=@n58, Translation="0,0,2", Rotation="20,0.000,-0.000", Scale="1,1,1") -> (Pose) {
    alphaInputType=Bool
    rotationMode=BMM_Additive
    rotationSpace=BCS_BoneSpace
    translationMode=BMM_Additive
    translationSpace=BCS_BoneSpace
    alphaBoolBlend {blendInTime=0.25, blendOption=Linear, blendOutTime=0.25}
    alphaScaleBiasClamp {bInterpResult=true}
    boneToModify {boneName=dyn_dress_left_b_master}
    rotation {pitch=20}
    translation {z=2}
   }
   Output Pose(Result=Component To Local(ComponentPose=@n60)) {layerGroup=None}
   Get bIsCrouchSprinting()
  ]}
 EventGraph {nodes [
   Delay(Duration=0.1)
   > ResetDynamics(InTeleportType=TeleportPhysics)
   > ResetDynamics(InTeleportType=ResetPhysics)
   Event BlueprintUpdateAnimation()
   > n1 = Sequence!(execute) -> (then_0, then_1, then_2:Exec, then_3:Exec, then_4:Exec)
   > n2 = Branch(Condition=Get bIsCrouching())
   n3 = GetOwnerSkeletalMeshComponent()
   n4 = BreakRotator(InRot=BreakTransform(InTransform=GetDeltaTransformFromRefPose(self=@n3, BoneName=thigh_l, BaseName=pelvis)).Rotation)
   > Set Thigh L Crouch Pitch Alpha(Thigh L Crouch Pitch Alpha:double=MapRangeClamped(Value=@n4.Pitch, InRangeA=20, InRangeB=45)) -> (Output_Get:double)
   > Set Thigh L Crouch Rot Alpha(Thigh L Crouch Rot Alpha:double=MapRangeClamped(Value=@n4.Yaw, InRangeA=-20, InRangeB=-100)) -> (Output_Get:double)
   > Set Thigh L Crouch Pitch In Alpha(Thigh L Crouch Pitch In Alpha:double=MapRangeClamped(Value=@n4, InRangeA=0, InRangeB=60)) -> (Output_Get:double)
   n5 = GetOwnerSkeletalMeshComponent()
   n6 = BreakRotator(InRot=BreakTransform(InTransform=GetDeltaTransformFromRefPose(self=@n5, BoneName=thigh_r, BaseName=pelvis)).Rotation)
   > Set Thigh R Crouch Rot Roll Alpha(Thigh R Crouch Rot Roll Alpha:double=MapRangeClamped(Value=@n6, InRangeA=0, InRangeB=40)) -> (Output_Get:double)
   > Set Thigh R Crouch Pitch Alpha(Thigh R Crouch Pitch Alpha:double=MapRangeClamped(Value=@n6.Pitch, InRangeA=15, InRangeB=35)) -> (Output_Get:double)
   > Set Thigh R Crouch Rot Alpha(Thigh R Crouch Rot Alpha:double=MapRangeClamped(Value=@n6.Yaw, InRangeA=-25, InRangeB=-105)) -> (Output_Get:double)
   > Set Thigh L Pitch Alpha(Thigh L Pitch Alpha:double=0) -> (Output_Get:double)
   > Set Thigh R Pitch Alpha(Thigh R Pitch Alpha:double=0) -> (Output_Get:double)
   > Set Thigh L Rot Alpha(Thigh L Rot Alpha:double=0) -> (Output_Get:double)
   > Set Thigh R Rot Alpha(Thigh R Rot Alpha:double=0) -> (Output_Get:double)
   Set Thigh L Crouch Pitch Alpha(execute=@n2.else, Thigh L Crouch Pitch Alpha=0)
   n7 = GetOwnerSkeletalMeshComponent()
   n8 = BreakRotator(InRot=BreakTransform(InTransform=GetDeltaTransformFromRefPose(self=@n7, BoneName=thigh_l, BaseName=pelvis)).Rotation)
   > Set Thigh L Pitch Alpha(Thigh L Pitch Alpha=MapRangeClamped(Value=@n8.Pitch, InRangeA=0, InRangeB=30))
   > Set Thigh L Crouch Rot Alpha(Thigh L Crouch Rot Alpha=0)
   > Set Thigh L Rot Alpha(Thigh L Rot Alpha=MapRangeClamped(Value=@n8.Yaw, InRangeA=-20, InRangeB=-80))
   > Set Thigh L Crouch Pitch In Alpha(Thigh L Crouch Pitch In Alpha=0)
   > Set Thigh R Crouch Rot Roll Alpha(Thigh R Crouch Rot Roll Alpha=0)
   > Set Thigh R Crouch Pitch Alpha(Thigh R Crouch Pitch Alpha=0)
   n9 = GetOwnerSkeletalMeshComponent()
   n10 = BreakRotator(InRot=BreakTransform(InTransform=GetDeltaTransformFromRefPose(self=@n9, BoneName=thigh_r, BaseName=pelvis)).Rotation)
   > Set Thigh R Pitch Alpha(Thigh R Pitch Alpha=MapRangeClamped(Value=@n10, InRangeA=15, InRangeB=30))
   > Set Thigh R Crouch Rot Alpha(Thigh R Crouch Rot Alpha=0)
   > Set Thigh R Rot Alpha(Thigh R Rot Alpha=MapRangeClamped(Value=@n10.Yaw, InRangeA=-25, InRangeB=-105))
   n11 = BreakTransform(InTransform=GetDeltaTransformFromRefPose(self=GetOwnerSkeletalMeshComponent(), BoneName=pelvis, BaseName=root))
   Set Pelvis Rot(execute=@n1.then_1, Pelvis Rot:double=MapRangeClamped(Value=BreakRotator(InRot=@n11.Rotation), InRangeA=-20, InRangeB=-50)) -> (Output_Get:double)
   Set Skydiving(
    execute=@n1.then_2
    Skydiving:bool=OR Boolean!(A=Get bIsSkyDiving(), B=Get bIsSkydiveFloating(), C:bool=Get bIsSkydiveDiveMode(), D:bool=Get bIsParachuteOpened()) -> (ReturnValue)
   ) -> (Output_Get:bool)
   n12 = Gravity Override State Function DEPRECATED(
    execute=@n1.then_3
    CharacterPartAnimInstance=Self-Reference()
    GameTimeInSeconds=GetGameTimeInSeconds()
    GravityOverrideParams=Get Gravity Override Params()
   )
   > Set GravityOverride(GravityOverride=@n12)
   n13 = Get bIsPlayingEmote()
   n14 = Set FrontendNotEmote(execute=@n1.then_4, FrontendNotEmote:bool=AND Boolean(A=NOT Boolean(A=@n13), B=Get bIsInFrontEnd())) -> (Output_Get:bool)
   n15 = Get bRigidBodyEnabled()
   > Set FrontendNotEmoteRbanDisabled(FrontendNotEmoteRbanDisabled:bool=AND Boolean(A=@n14, B=NOT Boolean(A=@n15))) -> (Output_Get:bool)
   MapRangeClamped(Value=@n10.Pitch, InRangeA=15, InRangeB=35)
   MapRangeClamped(Value=@n8, InRangeA=0, InRangeB=60)
   Event OnCharacterCustomizationLoaded()
  ]}
]
variables [
 FrontendNotEmote {type=boolean}
 FrontendNotEmoteRbanDisabled {type=boolean}
 Gravity Override Params {type=GravityOverrideParamsStruct}
 GravityOverride {type=Vector}
 Pelvis Rot {type=number}
 Skydiving {type=boolean}
 Thigh L Crouch Pitch Alpha {type=number}
 Thigh L Crouch Pitch In Alpha {type=number}
 Thigh L Crouch Rot Alpha {type=number}
 Thigh L Pitch Alpha {type=number}
 Thigh L Rot Alpha {type=number}
 Thigh R Crouch Pitch Alpha {type=number}
 Thigh R Crouch Rot Alpha {type=number}
 Thigh R Crouch Rot Roll Alpha {type=number}
 Thigh R Pitch Alpha {type=number}
 Thigh R Rot Alpha {type=number}
]
