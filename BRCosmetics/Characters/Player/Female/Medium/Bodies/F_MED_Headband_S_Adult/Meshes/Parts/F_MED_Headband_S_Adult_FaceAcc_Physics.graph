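# Physics asset for the FaceAcc part of the female medium headband cosmetic, defining bodies and joints for its dynamic hair bones.
# Five capsule chains (back, back-left, back-right, front-left, front-right) hang from a head capsule through small base bodies with free swing and twist.
# Joints along each chain set narrow angle limits, tighter swing in front than in back.
# Both front chains fork at their third bone into two short branches.
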PhysicsAsset /BRCosmetics/Characters/Player/Female/Medium/Bodies/F_MED_Headband_S_Adult/Meshes/Parts/F_MED_Headband_S_Adult_FaceAcc_Physics
bodies {
 dyn_hair_b_01 [{shapeType=Capsule, length=2.75, radius=3.25, shapeName=dyn_hair_b_01_capsule, rotation=<rot 90 0 0>}]
 dyn_hair_b_02 [{shapeType=Capsule, length=3.25, radius=2.75, shapeName=dyn_hair_b_02_capsule, rotation=<rot 90 0 0>}]
 dyn_hair_b_03 [{shapeType=Capsule, length=3.25, radius=2.75, shapeName=dyn_hair_b_03_capsule, rotation=<rot 90 0 0>}]
 dyn_hair_b_base [{shapeType=Capsule, length=0.5, radius=0.5, shapeName=dyn_hair_b_base_capsule, rotation=<rot 90 0 0>}]
 dyn_hair_bl_01 [{shapeType=Capsule, length=2.5, radius=2.5, shapeName=dyn_hair_bl_01_capsule, rotation=<rot 90 0 0>}]
 dyn_hair_bl_02 [{shapeType=Capsule, length=2.5, radius=2.5, shapeName=dyn_hair_bl_02_capsule, rotation=<rot 90 0 0>}]
 dyn_hair_bl_03 [{shapeType=Capsule, length=2.5, radius=2.5, shapeName=dyn_hair_bl_03_capsule, rotation=<rot 90 0 0>}]
 dyn_hair_bl_base [{shapeType=Capsule, length=0.5, radius=0.5, shapeName=dyn_hair_bl_base_capsule, rotation=<rot 90 0 0>}]
 dyn_hair_br_01 [{shapeType=Capsule, length=2.5, radius=2.5, shapeName=dyn_hair_br_01_capsule, rotation=<rot 90 0 0>}]
 dyn_hair_br_02 [{shapeType=Capsule, length=2.5, radius=2.5, shapeName=dyn_hair_br_02_capsule, rotation=<rot 90 0 0>}]
 dyn_hair_br_03 [{shapeType=Capsule, length=2.5, radius=2.5, shapeName=dyn_hair_br_03_capsule, rotation=<rot 90 0 0>}]
 dyn_hair_br_base [{shapeType=Capsule, length=0.5, radius=0.5, shapeName=dyn_hair_br_base_capsule, rotation=<rot 90 0 0>}]
 dyn_hair_fl_01 [{shapeType=Capsule, length=3, radius=2, shapeName=dyn_hair_fl_01_capsule, rotation=<rot 90 0 0>}]
 dyn_hair_fl_02 [{shapeType=Capsule, length=3, radius=2, shapeName=dyn_hair_fl_02_capsule, rotation=<rot 90 0 0>}]
 dyn_hair_fl_03 [{shapeType=Capsule, length=3, radius=2, shapeName=dyn_hair_fl_03_capsule, rotation=<rot 90 0 0>}]
 dyn_hair_fl_04 [{shapeType=Capsule, length=3, radius=2, shapeName=dyn_hair_fl_04_capsule, rotation=<rot 90 0 0>}]
 dyn_hair_fl_05 [{shapeType=Capsule, length=3, radius=2, shapeName=dyn_hair_fl_05_capsule, rotation=<rot 90 0 0>}]
 dyn_hair_fl_06 [{shapeType=Capsule, length=3, radius=2, shapeName=dyn_hair_fl_06_capsule, rotation=<rot 90 0 0>}]
 dyn_hair_fl_07 [{shapeType=Capsule, length=3, radius=2, shapeName=dyn_hair_fl_07_capsule, rotation=<rot 90 0 0>}]
 dyn_hair_fl_base [{shapeType=Capsule, length=0.5, radius=0.5, shapeName=dyn_hair_fl_base_capsule, rotation=<rot 90 0 0>}]
 dyn_hair_fr_01 [{shapeType=Capsule, length=3, radius=2, shapeName=dyn_hair_fr_01_capsule, rotation=<rot 90 0 0>}]
 dyn_hair_fr_02 [{shapeType=Capsule, length=3, radius=2, shapeName=dyn_hair_fr_02_capsule, rotation=<rot 90 0 0>}]
 dyn_hair_fr_03 [{shapeType=Capsule, length=3, radius=2, shapeName=dyn_hair_fr_03_capsule, rotation=<rot 90 0 0>}]
 dyn_hair_fr_04 [{shapeType=Capsule, length=3, radius=2, shapeName=dyn_hair_fr_04_capsule, rotation=<rot 90 0 0>}]
 dyn_hair_fr_05 [{shapeType=Capsule, length=0, radius=0.5, shapeName=dyn_hair_fr_05_capsule, rotation=<rot 90 0 0>}]
 dyn_hair_fr_06 [{shapeType=Capsule, length=3, radius=2, shapeName=dyn_hair_fr_06_capsule, rotation=<rot 90 0 0>}]
 dyn_hair_fr_07 [{shapeType=Capsule, length=3, radius=2, shapeName=dyn_hair_fr_07_capsule, rotation=<rot 90 0 0>}]
 dyn_hair_fr_base [{shapeType=Capsule, length=0.5, radius=0.5, shapeName=dyn_hair_fr_base_capsule, rotation=<rot 90 0 0>}]
 head [{shapeType=Capsule, length=6.72, radius=6.22, center=<2.769 0.75 0>, rotation=<rot 90 0 -13>}]
}
constraints [
 {
  bone1Name=dyn_hair_fr_base
  bone2Name=head
  swing1LimitDegrees=45
  swing1Motion=Free
  swing2LimitDegrees=45
  swing2Motion=Free
  twistLimitDegrees=45
  twistMotion=Free
 }
 {bone1Name=dyn_hair_fr_01, bone2Name=dyn_hair_fr_base, swing1LimitDegrees=8, swing2LimitDegrees=8, twistLimitDegrees=5}
 {bone1Name=dyn_hair_fr_02, bone2Name=dyn_hair_fr_01, swing1LimitDegrees=8, swing2LimitDegrees=8, twistLimitDegrees=5}
 {bone1Name=dyn_hair_fr_03, bone2Name=dyn_hair_fr_02, swing1LimitDegrees=8, swing2LimitDegrees=8, twistLimitDegrees=5}
 {bone1Name=dyn_hair_fr_04, bone2Name=dyn_hair_fr_03, swing1LimitDegrees=8, swing2LimitDegrees=8, twistLimitDegrees=5}
 {bone1Name=dyn_hair_fr_05, bone2Name=dyn_hair_fr_04, swing1LimitDegrees=8, swing2LimitDegrees=8, twistLimitDegrees=5}
 {bone1Name=dyn_hair_fr_06, bone2Name=dyn_hair_fr_03, swing1LimitDegrees=8, swing2LimitDegrees=8, twistLimitDegrees=5}
 {bone1Name=dyn_hair_fr_07, bone2Name=dyn_hair_fr_06, swing1LimitDegrees=8, swing2LimitDegrees=8, twistLimitDegrees=5}
 {
  bone1Name=dyn_hair_br_base
  bone2Name=head
  swing1LimitDegrees=45
  swing1Motion=Free
  swing2LimitDegrees=45
  swing2Motion=Free
  twistLimitDegrees=45
  twistMotion=Free
 }
 {bone1Name=dyn_hair_br_01, bone2Name=dyn_hair_br_base, swing1LimitDegrees=15, swing2LimitDegrees=15, twistLimitDegrees=5}
 {bone1Name=dyn_hair_br_02, bone2Name=dyn_hair_br_01, swing1LimitDegrees=15, swing2LimitDegrees=15, twistLimitDegrees=5}
 {bone1Name=dyn_hair_br_03, bone2Name=dyn_hair_br_02, swing1LimitDegrees=15, swing2LimitDegrees=15, twistLimitDegrees=5}
 {
  bone1Name=dyn_hair_b_base
  bone2Name=head
  swing1LimitDegrees=45
  swing1Motion=Free
  swing2LimitDegrees=45
  swing2Motion=Free
  twistLimitDegrees=45
  twistMotion=Free
 }
 {bone1Name=dyn_hair_b_01, bone2Name=dyn_hair_b_base, swing1LimitDegrees=15, swing2LimitDegrees=15, twistLimitDegrees=5}
 {bone1Name=dyn_hair_b_02, bone2Name=dyn_hair_b_01, swing1LimitDegrees=15, swing2LimitDegrees=15, twistLimitDegrees=5}
 {bone1Name=dyn_hair_b_03, bone2Name=dyn_hair_b_02, swing1LimitDegrees=15, swing2LimitDegrees=15, twistLimitDegrees=5}
 {
  bone1Name=dyn_hair_bl_base
  bone2Name=head
  swing1LimitDegrees=45
  swing1Motion=Free
  swing2LimitDegrees=45
  swing2Motion=Free
  twistLimitDegrees=45
  twistMotion=Free
 }
 {bone1Name=dyn_hair_bl_01, bone2Name=dyn_hair_bl_base, swing1LimitDegrees=15, swing2LimitDegrees=15, twistLimitDegrees=5}
 {bone1Name=dyn_hair_bl_02, bone2Name=dyn_hair_bl_01, swing1LimitDegrees=15, swing2LimitDegrees=15, twistLimitDegrees=5}
 {bone1Name=dyn_hair_bl_03, bone2Name=dyn_hair_bl_02, swing1LimitDegrees=15, swing2LimitDegrees=15, twistLimitDegrees=5}
 {
  bone1Name=dyn_hair_fl_base
  bone2Name=head
  swing1LimitDegrees=45
  swing1Motion=Free
  swing2LimitDegrees=45
  swing2Motion=Free
  twistLimitDegrees=45
  twistMotion=Free
 }
 {bone1Name=dyn_hair_fl_01, bone2Name=dyn_hair_fl_base, swing1LimitDegrees=8, swing2LimitDegrees=8, twistLimitDegrees=5}
 {bone1Name=dyn_hair_fl_02, bone2Name=dyn_hair_fl_01, swing1LimitDegrees=8, swing2LimitDegrees=8, twistLimitDegrees=5}
 {bone1Name=dyn_hair_fl_03, bone2Name=dyn_hair_fl_02, swing1LimitDegrees=8, swing2LimitDegrees=8, twistLimitDegrees=5}
 {bone1Name=dyn_hair_fl_04, bone2Name=dyn_hair_fl_03, swing1LimitDegrees=8, swing2LimitDegrees=8, twistLimitDegrees=5}
 {bone1Name=dyn_hair_fl_05, bone2Name=dyn_hair_fl_04, swing1LimitDegrees=8, swing2LimitDegrees=8, twistLimitDegrees=5}
 {bone1Name=dyn_hair_fl_06, bone2Name=dyn_hair_fl_03, swing1LimitDegrees=8, swing2LimitDegrees=8, twistLimitDegrees=5}
 {bone1Name=dyn_hair_fl_07, bone2Name=dyn_hair_fl_06, swing1LimitDegrees=8, swing2LimitDegrees=8, twistLimitDegrees=5}
]
